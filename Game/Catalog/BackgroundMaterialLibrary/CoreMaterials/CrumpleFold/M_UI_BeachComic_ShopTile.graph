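# Translucent UI material that draws a comic-style beach background for Item Shop tiles.
# Shared background parameters supply the colors, with a linear gradient, a comic texture and two adjustable palm silhouettes packed in one texture's channels.
# A star-field texture sampled in radial coordinates and masked by a radial gradient is blended in, with random grain multiplied over the composite.
# Rounded shape and outline masks read corner radius from the shop's parameter collection.
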Material /Game/Catalog/BackgroundMaterialLibrary/CoreMaterials/CrumpleFold/M_UI_BeachComic_ShopTile
graph {
 properties {bAutomaticallySetUsageInEditor=false, blendMode=BLEND_Translucent, materialDomain=MD_UI, usedShadingModels=Default Lit}
 nodes [
  Constant() {r=0.0016}
  MaterialFunctionCall!(Input Alpha=Constant() {r=1}, bOverrideCornerRadius, Corner Radius Override) {materialFunction=/FortUILibrary/Systems/ItemShopTile/Materials/MaterialFunctions/MF_UI_ItemShop_RoundedCorners}
  n1 = MaterialFunctionCall!(UV, Seed) {materialFunction=&MF_UI_Random}
  n2 = MaterialFunctionCall!(Vector or UVs=NamedRerouteUsage()) {materialFunction=/FortUILibrary/MaterialFunctions/Basics/MF_UI_CheaperVectorToRadial}
  n3 = Multiply(
   A=@n2.Radial Coordinates
   B=AppendVector(A=ScalarParameter() {defaultValue=0.4198, parameterName=G-U-Tiles}, B=ScalarParameter() {defaultValue=0.0018, parameterName=G-V-Tiles})
  )
  n4 = ScalarParameter() {defaultValue=10, parameterName=TextureSize}
  n5 = MaterialFunctionCall(UV=NamedRerouteUsage(), GradientStart=Constant(), GradientEnd=Constant() {r=0.55}) {materialFunction=&MF_UI_RadialGradient}
  n6 = Min(
   A=TextureSample(UVs=Multiply(A=@n3, B=@n4)) {texture=/Game/Catalog/BackgroundMaterialLibrary/Textures/T_UI_StarField}.G
   B=OneMinus(None=SmoothStep(Value=@n5.Smooth))
  )
  n7 = LinearInterpolate(
   A=LinearInterpolate(A=NamedRerouteUsage(), Alpha=Multiply(A=@n6) {constB=1.5}) {constB=0}
   Alpha=Multiply(A=NamedRerouteUsage()) {constB=0.4}
  ) {constB=0}
  n8 = Constant3Vector() {constant=<rgba 0.9131 0.8714 0.807 1>}
  n9 = NamedRerouteUsage()
  n10 = Max(A=NamedRerouteUsage(), B=NamedRerouteUsage())
  Multiply(
   A=Add(A=Multiply(A=@n1.Result) {constB=0.3})
   B=LinearInterpolate(A=LinearInterpolate(A=@n7, B=@n8, Alpha=@n9) {constB=0}, Alpha=Multiply(A=@n10) {constB=1.2}) {constB=0}
  )
  n11 = MaterialFunctionCall(UVs=NamedRerouteUsage(), Texture Scale=ScalarParameter() {defaultValue=1, group=PalmLeft, parameterName=Size}) {materialFunction=&ScaleUVsByCenter}
  n12 = AppendVector(A=ScalarParameter() {group=PalmLeft, parameterName=X}, B=ScalarParameter() {group=PalmLeft, parameterName=Y})
  n13 = TextureSample(UVs=MaterialFunctionCall(UVs=Add(A=@n11.UVs, B=@n12), "Rotation Angle (0-1)"=ScalarParameter() {group=PalmLeft, parameterName=LRotation}) {materialFunction=&CustomRotator}.Rotated Values) {samplerType=SAMPLERTYPE_LinearColor, texture=/Game/Catalog/BackgroundMaterialLibrary/CoreMaterials/CrumpleFold/Textures/T_UI_SummerTree}
  NamedRerouteDeclaration(None=@n13.R) {name=LeftPalm, nodeColor=<rgba 0 0.8828 1 1>}
  n14 = Add(A=MaterialFunctionCall() {materialFunction=&GetUserInterfaceUV}.Normalized UV, B=Constant2Vector() {r=0.5}) {constB=0}
  NamedRerouteDeclaration(None=Add(
    A=MaterialFunctionCall(UV=@n14) {materialFunction=&MF_UI_UVScaling}.Scaled to Fit
    B=Constant2Vector() {r=-0.5}
   )) {name=LeftPalmUV, nodeColor=<rgba 1 0.9375 0 1>}
  n15 = MaterialFunctionCall() {materialFunction=&MF_UI_UVScaling}
  NamedRerouteDeclaration(None=@n15.Scaled to Fit) {name=Scale Fit, nodeColor=<rgba 0 0 0 1>}
  n16 = MaterialFunctionCall() {materialFunction=/FortUILibrary/Systems/ItemShopTile/Materials/MaterialFunctions/MF_UI_ItemShop_BackgroundParams}
  NamedRerouteDeclaration(None=@n16.Color 1) {name=Primary Color, nodeColor=<rgba 0 0 0 1>}
  NamedRerouteDeclaration(None=@n16.Color 2) {name=Secondary Color, nodeColor=<rgba 0 0 0 1>}
  NamedRerouteDeclaration(None=@n16.Color 3) {name=Accent Color, nodeColor=<rgba 0 0 0 1>}
  NamedRerouteDeclaration(None=MaterialFunctionCall!(ScreenUV, TextureSize, ScalePivotPoint) {materialFunction=/FortUILibrary/MaterialFunctions/Basics/MF_UI_ScreenSpaceUVScaling}.ScaleX) {name=ScreenSpaceUV, nodeColor=<rgba 1 0.3984 0 1>}
  n17 = Add(A=MaterialFunctionCall() {materialFunction=&GetUserInterfaceUV}.Normalized UV, B=Constant2Vector() {r=-0.5}) {constB=0}
  NamedRerouteDeclaration(None=Add(
    A=MaterialFunctionCall(UV=@n17) {materialFunction=&MF_UI_UVScaling}.Scaled to Fit
    B=Constant2Vector() {r=0.5}
   )) {name=RigtPalmUV, nodeColor=<rgba 1 0.9375 0 1>}
  n18 = MaterialFunctionCall(UVs=NamedRerouteUsage(), Texture Scale=Constant() {r=2}) {materialFunction=&ScaleUVsByCenter}
  n19 = AppendVector(A=Constant() {r=-0.05}, B=Constant())
  NamedRerouteDeclaration(None=TextureSample(UVs=Add(A=@n18.UVs, B=@n19)) {texture=/Game/Catalog/BackgroundMaterialLibrary/CoreMaterials/CrumpleFold/Textures/T_UI_ComicTexture}.R) {name=RedChannel, nodeColor=<rgba 1 0.0527 0 1>}
  n20 = CollectionParameter() {collection=/FortUILibrary/Systems/ItemShopTile/MPC_ItemShop, parameterName=CornerRadius}
  n21 = MaterialFunctionCall(
   Corner Radius=MaterialFunctionCall(R=@n20, G=@n20, B=@n20, A=@n20) {materialFunction=&AppendMany}.RGBA
   Radius in Pixels=Constant() {r=1}
   Outline Thickness=Constant() {r=10}
  ) {materialFunction=&MF_UI_CoreUI_RoundedMask}
  n22 = OneMinus(None=@n21.Outline SDF)
  NamedRerouteDeclaration(None=LinearInterpolate(A=Multiply(A=@n22) {constB=0}, Alpha=@n21.Outline Mask)) {name=OutlineFrame, nodeColor=<rgba 0 0.5547 1 1>}
  NamedRerouteDeclaration(None=@n21.Shape Mask) {name=FlatBGMask, nodeColor=<rgba 0 0.7188 1 1>}
  NamedRerouteDeclaration(None=@n15.ScaleY) {name=Y UV, nodeColor=<rgba 1 0.375 0 1>}
  n23 = NamedRerouteUsage()
  n24 = Constant() {r=0.75}
  NamedRerouteDeclaration(None=LinearInterpolate(
    A=NamedRerouteUsage()
    B=NamedRerouteUsage()
    Alpha=MaterialFunctionCall(UVs=@n23, RotationAngle=@n24) {materialFunction=&MF_UI_LinearGradient}.Result
   )) {name=LinearGradient, nodeColor=<rgba 1 0.75 0 1>}
  NamedRerouteDeclaration(None=@n15.ScaleX) {name=Screen UV, nodeColor=<rgba 0 0 0 1>}
  n25 = MaterialFunctionCall(UVs=NamedRerouteUsage(), Texture Scale=ScalarParameter() {defaultValue=1, group=RightPalm, parameterName=SizeR}) {materialFunction=&ScaleUVsByCenter}
  n26 = AppendVector(A=ScalarParameter() {group=RightPalm, parameterName=X R}, B=ScalarParameter() {group=RightPalm, parameterName=Y R})
  n27 = TextureSample(UVs=MaterialFunctionCall(UVs=Add(A=@n25.UVs, B=@n26), "Rotation Angle (0-1)"=ScalarParameter() {group=RightPalm, parameterName=Rotation}) {materialFunction=&CustomRotator}.Rotated Values) {samplerType=SAMPLERTYPE_LinearColor, texture=&T_UI_SummerTree}
  NamedRerouteDeclaration(None=@n27.G) {name=RightPalm, nodeColor=<rgba 0.6875 1 0 1>}
  NamedRerouteDeclaration(None=@n15.Scaled To Fill) {name=ScaletoFill, nodeColor=<rgba 0 0.5078 1 1>}
  Panner(
   Coordinate=@n2.Radial Coordinates
   Speed=AppendVector(A=ScalarParameter() {parameterName=U-Speed}, B=ScalarParameter() {defaultValue=0.6, parameterName=V-Speed})
  )
  SmoothStep(Value=@n21.Outline SDF) {constMax=0.1}
 ]
}
