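# Translucent UI material for an item shop tile background named RedComic, layering animated comic texture shapes over a color gradient in a rounded frame.
# Color parameters come from the shared ItemShop background function, alongside linear-gradient and UV-scaling helpers.
# T_UI_ComicTexture supplies channel masks: green and blue rotate, scale and slowly pan, while red is scaled and offset without motion.
# A rounded-mask function driven by the ItemShop collection's CornerRadius builds background and glowing-outline masks for lerp-based compositing.
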Material /Game/Catalog/BackgroundMaterialLibrary/CoreMaterials/CrumpleFold/M_UI_RedComic_ShopTile
graph {
 properties {bAutomaticallySetUsageInEditor=false, blendMode=BLEND_Translucent, materialDomain=MD_UI, usedShadingModels=Default Lit}
 nodes [
  n1 = NamedRerouteUsage()
  n2 = NamedRerouteUsage()
  n3 = Multiply(A=LinearInterpolate(A=LinearInterpolate(Alpha=@n2) {constA=0.2, constB=0.6}, Alpha=NamedRerouteUsage()) {constB=0}, B=NamedRerouteUsage())
  n4 = NamedRerouteUsage()
  n5 = NamedRerouteUsage()
  n6 = NamedRerouteUsage()
  LinearInterpolate(
   A=LinearInterpolate(
    A=LinearInterpolate(A=@n1, B=@n3, Alpha=@n4)
    B=MaterialFunctionCall!(Base=@n5, Add=@n6) {materialFunction=/Engine/Functions/Engine_MaterialFunctions02/Math/MultiplyAdd}.Result
    Alpha=@n6
   )
   B=Constant3Vector() {constant=<rgba 0.7084 0.9216 1 1>}
   Alpha=NamedRerouteUsage()
  )
  MaterialFunctionCall!(Input Alpha=Constant() {r=1}, bOverrideCornerRadius, Corner Radius Override) {materialFunction=/FortUILibrary/Systems/ItemShopTile/Materials/MaterialFunctions/MF_UI_ItemShop_RoundedCorners}
  n7 = MaterialFunctionCall() {materialFunction=&MF_UI_UVScaling}
  NamedRerouteDeclaration(None=@n7.ScaleY) {name=Y UV, nodeColor=<rgba 1 0.375 0 1>}
  NamedRerouteDeclaration(None=@n7.ScaleX) {name=Screen UV, nodeColor=<rgba 0 0 0 1>}
  n8 = CollectionParameter() {collection=/FortUILibrary/Systems/ItemShopTile/MPC_ItemShop, parameterName=CornerRadius}
  n9 = MaterialFunctionCall(
   Corner Radius=MaterialFunctionCall(R=@n8, G=@n8, B=@n8, A=@n8) {materialFunction=&AppendMany}.RGBA
   Radius in Pixels=Constant() {r=1}
   Outline Thickness=Constant() {r=6}
  ) {materialFunction=&MF_UI_CoreUI_RoundedMask}
  n10 = OneMinus(None=@n9.Outline SDF)
  NamedRerouteDeclaration(None=LinearInterpolate(A=Multiply(A=@n10) {constB=0}, Alpha=@n9.Outline Mask)) {name=OutlineFrame, nodeColor=<rgba 0 0.5547 1 1>}
  NamedRerouteDeclaration(None=@n9.Shape Mask) {name=FlatBGMask, nodeColor=<rgba 0 0.7188 1 1>}
  n11 = MaterialFunctionCall() {materialFunction=/FortUILibrary/Systems/ItemShopTile/Materials/MaterialFunctions/MF_UI_ItemShop_BackgroundParams}
  NamedRerouteDeclaration(None=@n11.Color 1) {name=Primary Color, nodeColor=<rgba 0 0 0 1>}
  NamedRerouteDeclaration(None=@n11.Color 2) {name=Secondary Color, nodeColor=<rgba 0 0 0 1>}
  NamedRerouteDeclaration(None=@n11.Color 3) {name=Accent Color, nodeColor=<rgba 0 0 0 1>}
  NamedRerouteDeclaration(None=@n7.Scaled to Fit) {name=Scale Fit, nodeColor=<rgba 0 0 0 1>}
  NamedRerouteDeclaration(None=MaterialFunctionCall!(ScreenUV, TextureSize, ScalePivotPoint) {materialFunction=/FortUILibrary/MaterialFunctions/Basics/MF_UI_ScreenSpaceUVScaling}.ScaleX) {name=ScreenSpaceUV, nodeColor=<rgba 1 0.3984 0 1>}
  n12 = MaterialFunctionCall(UVs=NamedRerouteUsage(), "Rotation Angle (0-1)"=Constant() {r=0.1}) {materialFunction=&CustomRotator}
  n13 = Constant() {r=0.2}
  n14 = TextureSample(UVs=Panner(Coordinate=MaterialFunctionCall(UVs=@n12.Rotated Values, Texture Scale=@n13) {materialFunction=&ScaleUVsByCenter}.UVs) {speedX=0.005, speedY=0.007}) {texture=/Game/Catalog/BackgroundMaterialLibrary/CoreMaterials/CrumpleFold/Textures/T_UI_ComicTexture}
  NamedRerouteDeclaration(None=@n14.B) {name=BlueChannel, nodeColor=<rgba 0 0.0705 1 1>}
  n15 = Max(A=Max(A=NamedRerouteUsage(), B=NamedRerouteUsage()), B=NamedRerouteUsage())
  NamedRerouteDeclaration(None=Multiply(A=Saturate(None=@n15)) {constB=0.6}) {name=ShapeMask, nodeColor=<rgba 0 0.9297 1 1>}
  n16 = NamedRerouteUsage()
  n17 = Constant() {r=0.75}
  NamedRerouteDeclaration(None=LinearInterpolate(
    A=NamedRerouteUsage()
    B=NamedRerouteUsage()
    Alpha=MaterialFunctionCall(UVs=@n16, RotationAngle=@n17) {materialFunction=&MF_UI_LinearGradient}.Result
   )) {name=LinearGradient, nodeColor=<rgba 1 0.75 0 1>}
  NamedRerouteDeclaration(None=@n14.G) {name=GreenChannel, nodeColor=<rgba 0.2569 1 0 1>}
  n18 = MaterialFunctionCall(UVs=NamedRerouteUsage(), Texture Scale=Constant() {r=2}) {materialFunction=&ScaleUVsByCenter}
  n19 = AppendVector(A=Constant() {r=-0.05}, B=Constant())
  NamedRerouteDeclaration(None=TextureSample(UVs=Add(A=@n18.UVs, B=@n19)) {texture=&T_UI_ComicTexture}.R) {name=RedChannel, nodeColor=<rgba 1 0.0527 0 1>}
  SmoothStep(Value=@n9.Outline SDF) {constMax=0.1}
 ]
}
